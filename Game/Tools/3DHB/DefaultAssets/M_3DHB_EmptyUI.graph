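# M_3DHB_EmptyUI is a translucent User Interface domain material among the default assets of the 3DHB tool, named as an empty UI material.
# Its graph holds a single scalar constant, with no connection to any material output shown.
# It has no textures, parameters or other logic.
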Material /Game/Tools/3DHB/DefaultAssets/M_3DHB_EmptyUI
graph {
 properties {blendMode=BLEND_Translucent, materialDomain=MD_UI, pixelDepthOffsetMode=PDOM_Legacy, usedShadingModels=Default Lit}
 nodes [Constant() {r=0.2}]
}
