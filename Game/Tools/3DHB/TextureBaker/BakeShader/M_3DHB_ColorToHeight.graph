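# Translucent UI-domain material for the 3DHB texture baker that converts a color texture into a grayscale height value.
# The sampled texture is fully desaturated, blended toward its inverse by Invert, then offset by Add and saturated.
# A power curve with an exponent divided by the lower-clamped Brightness parameter shapes the result, which is finally multiplied by the texture's alpha.
# The texture is sampled in explicit mip-level mode.
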Material /Game/Tools/3DHB/TextureBaker/BakeShader/M_3DHB_ColorToHeight
graph {
 properties {blendMode=BLEND_Translucent, materialDomain=MD_UI, pixelDepthOffsetMode=PDOM_Legacy, shadingModel=MSM_Unlit, usedShadingModels=Unlit}
 nodes [
  n1 = TextureSample!(
   UVs
   Tex=TextureObjectParameter() {parameterName=Texture, texture=/Game/Tools/3DHB/DefaultAssets/T_3DHB_DefaultTexture}
   Level
   Apply View MipBias
  ) {mipValueMode=TMVM_MipLevel}
  n2 = Desaturation(None=@n1.RGB, Fraction=Constant() {r=1})
  n3 = Add(A=LinearInterpolate(A=@n2, B=OneMinus(None=@n2), Alpha=ScalarParameter() {parameterName=Invert}), B=ScalarParameter() {parameterName=Add}) {constB=0.01}
  n4 = Clamp(None=ScalarParameter() {defaultValue=1, parameterName=Brightness}) {clampMode=CMODE_ClampMin}
  Multiply(A=Power(Base=Saturate(None=@n3), Exp=Divide(B=@n4)), B=@n1.A)
 ]
}
