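# Translucent UI-domain material in the 3DHB TextureBaker tool that renders a Gaussian-blurred, single-channel grayscale version of an input texture.
# MF_3DHB_GaussianBlur samples the Texture parameter using Samples, DeNoise, BlurScale and a Power equal to the reciprocal of min-clamped Brightness.
# Only BlurScale's red component is used, so both axes blur equally.
# The blurred red and alpha channels are each spread to RGB, and UseAChannel blends between them.
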Material /Game/Tools/3DHB/TextureBaker/BakeShader/M_3DHB_GaussianBlur
graph {
 properties {blendMode=BLEND_Translucent, materialDomain=MD_UI, pixelDepthOffsetMode=PDOM_Legacy, shadingModel=MSM_Unlit, usedShadingModels=Unlit}
 nodes [
  n1 = VectorParameter() {parameterName=BlurScale, defaultValue=<rgba 1 1 0 1>}
  n2 = Clamp(None=ScalarParameter() {defaultValue=1, parameterName=Brightness}) {clampMode=CMODE_ClampMin}
  n3 = MaterialFunctionCall(Float4=MaterialFunctionCall!(
    BlurScale=AppendVector(A=@n1.R, B=@n1.R)
    Samples=ScalarParameter() {defaultValue=2, parameterName=Samples}
    Power=Divide(B=@n2)
    DeNoise=ScalarParameter() {parameterName=DeNoise}
    Texture=TextureObjectParameter() {parameterName=Texture, texture=/Game/Tools/3DHB/DefaultAssets/T_3DHB_DefaultTexture}
   ) {materialFunction=&/MF_3DHB_GaussianBlur}.Result) {materialFunction=&BreakOutFloat4Components}
  LinearInterpolate(
   A=MaterialFunctionCall(X=@n3.R, Y=@n3.R, Z=@n3.R) {materialFunction=&MakeFloat3}.Result
   B=MaterialFunctionCall(X=@n3.A, Y=@n3.A, Z=@n3.A) {materialFunction=&MakeFloat3}.Result
   Alpha=ScalarParameter() {parameterName=UseAChannel}
  )
 ]
}
