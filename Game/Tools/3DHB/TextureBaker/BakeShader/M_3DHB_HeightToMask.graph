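# M_3DHB_HeightToMask is an unlit, translucent UI-domain material in the 3DHB texture baker tool that derives a mask from captured height data.
# It samples the RT_3DHB_DepthCaptureOut render target as linear color using explicit mip-level sampling.
# The RGB is fully desaturated to grayscale, then passed through Ceil, so any fractional value above zero rounds up to one, giving a hard-edged mask.
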Material /Game/Tools/3DHB/TextureBaker/BakeShader/M_3DHB_HeightToMask
graph {
 properties {blendMode=BLEND_Translucent, materialDomain=MD_UI, pixelDepthOffsetMode=PDOM_Legacy, shadingModel=MSM_Unlit, usedShadingModels=Unlit}
 nodes [
  n1 = TextureSample!(UVs, Tex, Level, Apply View MipBias) {mipValueMode=TMVM_MipLevel, samplerType=SAMPLERTYPE_LinearColor, texture=/Game/Tools/3DHB/TextureBaker/RenderTarget/RT_3DHB_DepthCaptureOut}
  Ceil(None=Desaturation(None=@n1.RGB, Fraction=Constant() {r=1}))
 ]
}
