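# UI-domain material for the 3DHB texture baker that converts a height map into a normal map.
# It samples RT_3DHB_HTN_Blur one texel either side in U and V, sized by ResU and ResV, to get central-difference slopes.
# NormalStrength scales the slopes, with Y negated, then the vector gets Z=1, is normalized and is packed into 0–1.
# The aspect-ratio correction compares ResU with itself, so it has no effect.
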Material /Game/Tools/3DHB/TextureBaker/BakeShader/M_3DHB_HeightToNormal
graph {
 properties {blendMode=BLEND_Translucent, materialDomain=MD_UI, pixelDepthOffsetMode=PDOM_Legacy, shadingModel=MSM_Unlit, usedShadingModels=Unlit}
 nodes [
  n1 = TextureCoordinate()
  n2 = AppendVector(A=ScalarParameter() {defaultValue=128, parameterName=ResU}, B=ScalarParameter() {defaultValue=128, parameterName=ResV})
  n3 = MaterialFunctionCall(Float2=Divide(A=Constant2Vector() {g=1, r=1}, B=@n2)) {materialFunction=&BreakOutFloat2Components}
  n4 = Constant()
  n5 = MaterialFunctionCall(X=@n3.R, Y=@n4) {materialFunction=&MakeFloat2}
  n6 = TextureSample!(UVs=Subtract(A=@n1, B=@n5.Result), Tex, Level, Apply View MipBias) {mipValueMode=TMVM_MipLevel, samplerType=SAMPLERTYPE_LinearColor, texture=/Game/Tools/3DHB/TextureBaker/RenderTarget/RT_3DHB_HTN_Blur}
  n7 = TextureSample(UVs=Add(A=@n1, B=@n5.Result)) {mipValueMode=TMVM_MipLevel, samplerType=SAMPLERTYPE_LinearColor, texture=&RT_3DHB_HTN_Blur}
  n8 = MaterialFunctionCall(X=@n4, Y=@n3.G) {materialFunction=&MakeFloat2}
  n9 = TextureSample(UVs=Add(A=@n1, B=@n8.Result)) {mipValueMode=TMVM_MipLevel, samplerType=SAMPLERTYPE_LinearColor, texture=&RT_3DHB_HTN_Blur}
  n10 = TextureSample(UVs=Subtract(A=@n1, B=@n8.Result)) {mipValueMode=TMVM_MipLevel, samplerType=SAMPLERTYPE_LinearColor, texture=&RT_3DHB_HTN_Blur}
  n11 = MaterialFunctionCall(X=Subtract(A=@n6.R, B=@n7.R), Y=Subtract(A=@n9.R, B=@n10.R)) {materialFunction=&MakeFloat2}
  n12 = MaterialFunctionCall(Float2=ScalarParameter() {defaultValue=3, parameterName=NormalStrength}) {materialFunction=&BreakOutFloat2Components}
  n13 = MaterialFunctionCall(X=@n12.R, Y=Multiply(A=@n12.G) {constB=-1}) {materialFunction=&MakeFloat2}
  n14 = MaterialFunctionCall(Float2=@n2) {materialFunction=&BreakOutFloat2Components}
  n15 = MaterialFunctionCall(X=Divide(A=Min(A=@n14.R, B=@n14.R), B=Max(A=@n14.R, B=@n14.R)), Y=Constant() {r=1}) {materialFunction=&MakeFloat2}
  n16 = MaterialFunctionCall(Float2=@n15.Result) {materialFunction=&BreakOutFloat2Components}
  n17 = MaterialFunctionCall(X=@n16.G, Y=@n16.G) {materialFunction=&MakeFloat2}
  n18 = AppendVector(
   A=Multiply(A=Multiply(A=@n11.Result, B=@n13.Result), B=If(A=@n14.R, B=@n14.R, "A > B"=@n15.Result, "A == B"=@n15.Result, "A < B"=@n17.Result))
   B=Constant() {r=1}
  )
  n19 = Constant() {r=0.5}
  n20 = SquareRoot(None=DotProduct(A=@n18, B=@n18))
  Add(A=Multiply(A=@n18, B=Divide(A=@n19, B=@n20)), B=Constant() {r=0.5})
  Constant() {r=1}
 ]
}
