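# Translucent UI material for the 3DHB texture baker that merges diffuse and unlit capture render targets and color-grades the result.
# An Unlit scalar blends the captures, and the diffuse capture's alpha is inverted.
# Two chained selective tweaks (A, B) mask pixels by normalized-color distance to a target color, then apply desaturation, hue shift, multiply and add.
# A global pass follows with desaturation, hue shift, contrast, multiply and add.
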Material /Game/Tools/3DHB/TextureBaker/BakeShader/M_3DHB_MergeDiffuseCapture
graph {
 properties {blendMode=BLEND_Translucent, materialDomain=MD_UI, pixelDepthOffsetMode=PDOM_Legacy, shadingModel=MSM_Unlit, usedShadingModels=Unlit}
 nodes [
  n1 = TextureSample!(UVs, Tex, Level, Apply View MipBias) {
   mipValueMode=TMVM_MipLevel
   samplerType=SAMPLERTYPE_LinearColor
   texture=/Game/Tools/3DHB/TextureBaker/RenderTarget/RT_3DHB_DiffuseCaptureOutOrigin
  }
  OneMinus(None=@n1.A)
  n2 = TextureSample() {mipValueMode=TMVM_MipLevel, samplerType=SAMPLERTYPE_LinearColor, texture=/Game/Tools/3DHB/TextureBaker/RenderTarget/RT_3DHB_UnlitCaptureOut}
  n3 = LinearInterpolate(A=@n1.RGB, B=@n2.RGB, Alpha=ScalarParameter() {parameterName=Unlit})
  n4 = Distance(
   A=Subtract(
    A=Normalize(VectorInput=@n3)
    B=VectorParameter() {group=ColorTweakB, parameterName="ColorTweakA_TargetColor(Hue)", defaultValue=<rgba 0 0 0 1>}.RGB
   )
   B=Constant()
  )
  n5 = ScalarParameter() {group=ColorTweakB, parameterName=ColorTweakA_RangeOffset}
  n6 = Clamp(None=Subtract(A=OneMinus(None=@n4), B=@n5)) {clampMode=CMODE_ClampMin}
  n7 = Divide(B=Subtract(B=@n5))
  n8 = Saturate(None=Multiply(A=Multiply(A=@n6, B=@n7), B=ScalarParameter() {defaultValue=1, group=ColorTweakB, parameterName=ColorTweakA_RangeSharpness}))
  n9 = ScalarParameter() {group=ColorTweakB, parameterName=ColorTweakA_Desaturation}
  n10 = MaterialFunctionCall(
   Hue Shift Percentage=Multiply(A=ScalarParameter() {group=ColorTweakB, parameterName=ColorTweakA_HueShift}, B=@n8)
   Texture=Desaturation(None=@n3, Fraction=Multiply(A=@n8, B=@n9))
  ) {materialFunction=&HueShift}
  n11 = LinearInterpolate(B=VectorParameter() {group=ColorTweakB, parameterName=ColorTweakA_MulColor, defaultValue=<rgba 1 1 1 1>}.RGB, Alpha=@n8) {constA=1}
  n12 = VectorParameter() {group=ColorTweakB, parameterName=ColorTweakA_Add, defaultValue=<rgba 0 0 0 1>}
  n13 = Saturate(None=Add(A=Multiply(A=@n10.Result, B=@n11), B=LinearInterpolate(B=@n12.RGB, Alpha=@n8)))
  n14 = Distance(
   A=Subtract(
    A=Normalize(VectorInput=@n13)
    B=VectorParameter() {group=ColorTweakB, parameterName="ColorTweakB_TargetColor(Hue)", defaultValue=<rgba 0 0 0 1>}.RGB
   )
   B=Constant()
  )
  n15 = ScalarParameter() {group=ColorTweakB, parameterName=ColorTweakB_RangeOffset}
  n16 = Clamp(None=Subtract(A=OneMinus(None=@n14), B=@n15)) {clampMode=CMODE_ClampMin}
  n17 = Divide(B=Subtract(B=@n15))
  n18 = Saturate(None=Multiply(A=Multiply(A=@n16, B=@n17), B=ScalarParameter() {defaultValue=1, group=ColorTweakB, parameterName=ColorTweakB_RangeSharpness}))
  n19 = ScalarParameter() {group=ColorTweakB, parameterName=ColorTweakB_Desaturation}
  n20 = MaterialFunctionCall(
   Hue Shift Percentage=Multiply(A=ScalarParameter() {group=ColorTweakB, parameterName=ColorTweakB_HueShift}, B=@n18)
   Texture=Desaturation(None=@n13, Fraction=Multiply(A=@n18, B=@n19))
  ) {materialFunction=&HueShift}
  n21 = LinearInterpolate(B=VectorParameter() {group=ColorTweakB, parameterName=ColorTweakB_MulColor, defaultValue=<rgba 1 1 1 1>}.RGB, Alpha=@n18) {constA=1}
  n22 = VectorParameter() {group=ColorTweakB, parameterName=ColorTweakB_Add, defaultValue=<rgba 0 0 0 1>}
  n23 = Saturate(None=Add(A=Multiply(A=@n20.Result, B=@n21), B=LinearInterpolate(B=@n22.RGB, Alpha=@n18)))
  n24 = ScalarParameter() {group=ColorTweakGlobal, parameterName=ColorTweakGlobal_Desaturation}
  n25 = MaterialFunctionCall!(
   In=MaterialFunctionCall(
    Hue Shift Percentage=ScalarParameter() {group=ColorTweakGlobal, parameterName=ColorTweakGlobal_HueShift}
    Texture=Desaturation(None=@n23, Fraction=@n24)
   ) {materialFunction=&HueShift}.Result
   Contrast=ScalarParameter() {group=ColorTweakGlobal, parameterName=ColorTweakGlobal_Contrast}
  ) {materialFunction=/Engine/Functions/Engine_MaterialFunctions01/ImageAdjustment/CheapContrast_RGB}
  n26 = VectorParameter() {group=ColorTweakGlobal, parameterName=ColorTweakGlobal_MulColor, defaultValue=<rgba 1 1 1 1>}
  Power(
   Base=Add(
    A=Multiply(A=@n25.Result, B=@n26.RGB)
    B=VectorParameter() {group=ColorTweakGlobal, parameterName=ColorTweakGlobal_AddColor, defaultValue=<rgba 0 0 0 1>}.RGB
   )
   Exp=Constant() {r=1}
  )
 ]
}
