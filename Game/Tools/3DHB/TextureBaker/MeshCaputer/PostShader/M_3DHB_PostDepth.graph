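# Post-process material for the 3DHB TextureBaker mesh capture tool that renders scene depth as inverted greyscale.
# Scaled SceneDepth is remapped between Remap_Near and Remap_Far, and out-of-range pixels turn magenta unless they lie beyond DebugMaxDistance.
# A Brightness and Contrast chain exists but does not feed the output.
# A ShadingPathSwitch substitutes the T_3DHB_ES31 texture on Forward and Mobile paths.
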Material /Game/Tools/3DHB/TextureBaker/MeshCaputer/PostShader/M_3DHB_PostDepth
graph {
 properties {materialDomain=MD_PostProcess, pixelDepthOffsetMode=PDOM_Legacy, usedShadingModels=Default Lit}
 nodes [
  n1 = Multiply(A=ScalarParameter() {parameterName=Brightness}) {constB=-1}
  n2 = Clamp(None=@n1) {clampMode=CMODE_ClampMin}
  n3 = SceneTexture() {sceneTextureId=PPI_SceneDepth}
  n4 = Constant() {r=1000}
  n5 = ScalarParameter() {parameterName=Remap_Near}
  n6 = ScalarParameter() {defaultValue=1, parameterName=Remap_Far}
  n7 = OneMinus(None=MaterialFunctionCall(
    Input=Divide(A=@n3.Color, B=@n4)
    Input Low=Saturate(None=@n5)
    Input High=Saturate(None=@n6)
    Target Low=Constant()
    Target High=Constant() {r=1}
   ) {materialFunction=&RemapValueRange}.Result)
  n8 = MaterialFunctionCall(X=@n7, Y=@n7, Z=@n7) {materialFunction=&MakeFloat3}
  n9 = OneMinus(None=Power(Base=@n8.Result, Exp=Add(A=@n2)))
  n10 = Add(A=Abs(None=Clamp(None=@n1) {clampMode=CMODE_ClampMax, maxDefault=0}))
  MaterialFunctionCall(In=OneMinus(None=Power(Base=@n9, Exp=@n10)), Contrast=Max(A=ScalarParameter() {parameterName=Contrast}) {constB=0}) {materialFunction=&CheapContrast}
  n11 = Constant()
  n12 = ComponentMask(None=@n7) {r=true}
  n13 = Constant()
  n14 = Constant() {r=1}
  n15 = Constant()
  n16 = Constant() {r=1}
  n17 = Add(
   A=If(
    A=ComponentMask(None=@n3.Color) {r=true}
    B=ScalarParameter() {defaultValue=100000, parameterName=DebugMaxDistance}
    "A > B"=Constant() {r=1}
    "A == B"=@n11
    "A < B"=@n11
   )
   B=Min(A=If(A=@n12, B=@n13, "A > B"=@n14, "A == B"=@n14, "A < B"=@n15), B=If(A=@n12, B=@n16, "A > B"=@n15, "A == B"=@n14, "A < B"=@n14))
  )
  n18 = TextureSample() {texture=/Game/Tools/3DHB/DefaultAssets/T_3DHB_ES31}
  ShadingPathSwitch(
   Default=LinearInterpolate(A=Constant3Vector() {constant=<rgba 1 0 0.685 0>}, B=@n8.Result, Alpha=Saturate(None=@n17))
   Forward=@n18.RGB
   Mobile=@n18.RGB
  )
 ]
}
