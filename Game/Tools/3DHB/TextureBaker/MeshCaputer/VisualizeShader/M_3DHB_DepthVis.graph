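# Translucent UI material for the 3DHB TextureBaker tool that, by its name and folder, visualizes the mesh capturer's depth output.
# Its graph holds a single Texture Sample of the RT_3DHB_DepthCaptureOut render target, read as linear color with explicit mip-level mode.
# No depth remapping or other math nodes are present.
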Material /Game/Tools/3DHB/TextureBaker/MeshCaputer/VisualizeShader/M_3DHB_DepthVis
graph {
 properties {blendMode=BLEND_Translucent, materialDomain=MD_UI, pixelDepthOffsetMode=PDOM_Legacy, usedShadingModels=Default Lit}
 nodes [TextureSample!(UVs, Tex, Level, Apply View MipBias) {mipValueMode=TMVM_MipLevel, samplerType=SAMPLERTYPE_LinearColor, texture=/Game/Tools/3DHB/TextureBaker/RenderTarget/RT_3DHB_DepthCaptureOut}]
}
