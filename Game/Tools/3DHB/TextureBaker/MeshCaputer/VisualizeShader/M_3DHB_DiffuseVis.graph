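# Translucent UI-domain material in the 3DHB TextureBaker tool's mesh capturer, apparently for visualizing the captured diffuse result.
# It samples the RT_3DHB_DiffuseCaptureOut render target as linear color, using explicit mip-level sampling mode.
# A Power node raises the sampled RGB to the RGBCurver scalar parameter, an adjustable tone curve that is neutral by default.
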Material /Game/Tools/3DHB/TextureBaker/MeshCaputer/VisualizeShader/M_3DHB_DiffuseVis
graph {
 properties {blendMode=BLEND_Translucent, materialDomain=MD_UI, pixelDepthOffsetMode=PDOM_Legacy, usedShadingModels=Default Lit}
 nodes [
  n1 = TextureSample!(UVs, Tex, Level, Apply View MipBias) {
   mipValueMode=TMVM_MipLevel
   samplerType=SAMPLERTYPE_LinearColor
   texture=/Game/Tools/3DHB/TextureBaker/RenderTarget/RT_3DHB_DiffuseCaptureOut
  }
  Power(Base=@n1.RGB, Exp=ScalarParameter() {defaultValue=1, parameterName=RGBCurver})
 ]
}
