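# Niagara utility system for the 3DHB TextureBaker tool that initializes the size and format of its bake and capture render targets.
# One emitter's spawn script repeats a local InitialRenderTarget module for each user-bound target, passing shared ResU/ResV resolution through RenderTarget2D data interfaces.
# Format overrides use RGBA16f, except sRGB RGBA8 for diffuse capture and height blur.
# No particle spawner is present, and Capture_DORT has no default target.
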NiagaraSystem /Game/Tools/3DHB/TextureBaker/Niagara/NS_3DHB_InitialRenderTarget
dependencies {
 usedDataInterfaces [/Script/Niagara.NiagaraDataInterfaceRenderTarget2D]
 usedDynamicInputs []
 usedModules [&SystemState]
}
emitters {InitialRTSizeAndFormat {
  emitterSpawnScript [
   &self:InitialRenderTarget(
    ResU:NiagaraInt32=$User.ResU
    ResV:NiagaraInt32=$User.ResV
    TargetRT=DataInterface() {propertyValues {
      MipMapGeneration=Disabled
      MipMapGenerationType=Linear
      OverrideRenderTargetFilter=TF_Default
      OverrideRenderTargetFormat=RTF_RGBA16f
      bInheritUserParameterSettings=false
      bOverrideFormat=true
      bPreviewRenderTarget=false
      RenderTargetUserParameter {parameter {name=User.HeightRT, typeDefHandle {registeredTypeIndex=4}, varData []}}
      Size=<0 0>
     }}
   ) {module=InitialRenderTarget}
   &self:InitialRenderTarget(
    ResU:NiagaraInt32=$User.ResU
    ResV:NiagaraInt32=$User.ResV
    TargetRT=DataInterface() {propertyValues {
      MipMapGeneration=Disabled
      MipMapGenerationType=Linear
      OverrideRenderTargetFilter=TF_Default
      OverrideRenderTargetFormat=RTF_RGBA16f
      bInheritUserParameterSettings=false
      bOverrideFormat=true
      bPreviewRenderTarget=false
      RenderTargetUserParameter {parameter {name=User.HTN_B RT, typeDefHandle {registeredTypeIndex=4}, varData []}}
      Size=<0 0>
     }}
   ) {module=InitialRenderTarget001}
   &self:InitialRenderTarget(
    ResU:NiagaraInt32=$User.ResU
    ResV:NiagaraInt32=$User.ResV
    TargetRT=DataInterface() {propertyValues {
      MipMapGeneration=Disabled
      MipMapGenerationType=Linear
      OverrideRenderTargetFilter=TF_Default
      OverrideRenderTargetFormat=RTF_RGBA16f
      bInheritUserParameterSettings=false
      bOverrideFormat=true
      bPreviewRenderTarget=false
      RenderTargetUserParameter {parameter {name=User.MeshMaskRT, typeDefHandle {registeredTypeIndex=4}, varData []}}
      Size=<0 0>
     }}
   ) {module=InitialRenderTarget003}
   &self:InitialRenderTarget(
    ResU:NiagaraInt32=$User.ResU
    ResV:NiagaraInt32=$User.ResV
    TargetRT=DataInterface() {propertyValues {
      MipMapGeneration=Disabled
      MipMapGenerationType=Linear
      OverrideRenderTargetFilter=TF_Default
      OverrideRenderTargetFormat=RTF_RGBA16f
      bInheritUserParameterSettings=false
      bOverrideFormat=true
      bPreviewRenderTarget=false
      RenderTargetUserParameter {parameter {name=User.NormalRT, typeDefHandle {registeredTypeIndex=4}, varData []}}
      Size=<0 0>
     }}
   ) {module=InitialRenderTarget004}
   &self:InitialRenderTarget(
    ResU:NiagaraInt32=$User.ResU
    ResV:NiagaraInt32=$User.ResV
    TargetRT=DataInterface() {propertyValues {
      MipMapGeneration=Disabled
      MipMapGenerationType=Linear
      OverrideRenderTargetFilter=TF_Default
      OverrideRenderTargetFormat=RTF_RGBA8_SRGB
      bInheritUserParameterSettings=false
      bOverrideFormat=true
      bPreviewRenderTarget=false
      RenderTargetUserParameter {parameter {name=User.Capture_CRT, typeDefHandle {registeredTypeIndex=4}, varData []}}
      Size=<0 0>
     }}
   ) {module=InitialRenderTarget002}
   &self:InitialRenderTarget(
    ResU:NiagaraInt32=$User.ResU
    ResV:NiagaraInt32=$User.ResV
    TargetRT=DataInterface() {propertyValues {
      MipMapGeneration=Disabled
      MipMapGenerationType=Linear
      OverrideRenderTargetFilter=TF_Default
      OverrideRenderTargetFormat=RTF_RGBA16f
      bInheritUserParameterSettings=false
      bOverrideFormat=true
      bPreviewRenderTarget=false
      RenderTargetUserParameter {parameter {name=User.Capture_DRT, typeDefHandle {registeredTypeIndex=4}, varData []}}
      Size=<0 0>
     }}
   ) {module=InitialRenderTarget005}
   &self:InitialRenderTarget(
    ResU:NiagaraInt32=$User.ResU
    ResV:NiagaraInt32=$User.ResV
    TargetRT=DataInterface() {propertyValues {
      MipMapGeneration=Disabled
      MipMapGenerationType=Linear
      OverrideRenderTargetFilter=TF_Default
      OverrideRenderTargetFormat=RTF_RGBA16f
      bInheritUserParameterSettings=false
      bOverrideFormat=true
      bPreviewRenderTarget=false
      RenderTargetUserParameter {parameter {name=User.Capture_NRT, typeDefHandle {registeredTypeIndex=4}, varData []}}
      Size=<0 0>
     }}
   ) {module=InitialRenderTarget006}
   &self:InitialRenderTarget(
    ResU:NiagaraInt32=$User.ResU
    ResV:NiagaraInt32=$User.ResV
    TargetRT=DataInterface() {propertyValues {
      MipMapGeneration=Disabled
      MipMapGenerationType=Linear
      OverrideRenderTargetFilter=TF_Default
      OverrideRenderTargetFormat=RTF_RGBA8_SRGB
      bInheritUserParameterSettings=false
      bOverrideFormat=true
      bPreviewRenderTarget=false
      RenderTargetUserParameter {parameter {name=User.HeightBlurRT, typeDefHandle {registeredTypeIndex=4}, varData []}}
      Size=<0 0>
     }}
   ) {module=InitialRenderTarget007}
   &self:InitialRenderTarget(
    ResU:NiagaraInt32=$User.ResU
    ResV:NiagaraInt32=$User.ResV
    TargetRT=DataInterface() {propertyValues {
      MipMapGeneration=Disabled
      MipMapGenerationType=Linear
      OverrideRenderTargetFilter=TF_Default
      OverrideRenderTargetFormat=RTF_RGBA16f
      bInheritUserParameterSettings=false
      bOverrideFormat=true
      bPreviewRenderTarget=false
      RenderTargetUserParameter {parameter {name=User.Capture_URT, typeDefHandle {registeredTypeIndex=4}, varData []}}
      Size=<0 0>
     }}
   ) {module=InitialRenderTarget008}
   &self:InitialRenderTarget(
    ResU:NiagaraInt32=$User.ResU
    ResV:NiagaraInt32=$User.ResV
    TargetRT=DataInterface() {propertyValues {
      MipMapGeneration=Disabled
      MipMapGenerationType=Linear
      OverrideRenderTargetFilter=TF_Default
      OverrideRenderTargetFormat=RTF_RGBA16f
      bInheritUserParameterSettings=false
      bOverrideFormat=true
      bPreviewRenderTarget=false
      RenderTargetUserParameter {parameter {name=User.Capture_DORT, typeDefHandle {registeredTypeIndex=4}, varData []}}
      Size=<0 0>
     }}
   ) {module=InitialRenderTarget009}
  ]
  emitterUpdateScript [&EmitterState(Inactive Response=::NewEnumerator1, Life Cycle Mode=::NewEnumerator1, Loop Behavior=::NewEnumerator1, Loop Duration=0.3, UseLoopDelay=false)]
  particleSpawnScript [&InitializeParticle(Color Mode=::NewEnumerator0, Lifetime=5, UsePositionOffset=false, Write Material Random=true)]
  particleUpdateScript [&ParticleState()]
  rendererClasses [&NiagaraSpriteRendererProperties]
  renderers [{rendererClass=&NiagaraSpriteRendererProperties}]
 }}
summary {
 systemName=NS_3DHB_InitialRenderTarget
 emitters [{emitterName=InitialRTSizeAndFormat, rendererClasses [&NiagaraSpriteRendererProperties]}]
 userVariables [
  User.Capture_CRT {
   type=TextureRenderTarget
   description=""
   defaultValue {
    struct=NiagaraExt_VariableValue_Object
    value {object=/Game/Tools/3DHB/TextureBaker/RenderTarget/RT_3DHB_DiffuseCaptureOut, objectClass=/Script/Engine.TextureRenderTarget}
   }
  }
  User.Capture_DORT {
   type=TextureRenderTarget
   description=""
   defaultValue {struct=NiagaraExt_VariableValue_Object, value {object=null, objectClass=&TextureRenderTarget}}
  }
  User.Capture_DRT {
   type=TextureRenderTarget
   description=""
   defaultValue {
    struct=NiagaraExt_VariableValue_Object
    value {object=/Game/Tools/3DHB/TextureBaker/RenderTarget/RT_3DHB_DepthCaptureOut, objectClass=&TextureRenderTarget}
   }
  }
  User.Capture_NRT {
   type=TextureRenderTarget
   description=""
   defaultValue {
    struct=NiagaraExt_VariableValue_Object
    value {object=/Game/Tools/3DHB/TextureBaker/RenderTarget/RT_3DHB_NormalCaptureOut, objectClass=&TextureRenderTarget}
   }
  }
  User.Capture_URT {
   type=TextureRenderTarget
   description=""
   defaultValue {
    struct=NiagaraExt_VariableValue_Object
    value {object=/Game/Tools/3DHB/TextureBaker/RenderTarget/RT_3DHB_UnlitCaptureOut, objectClass=&TextureRenderTarget}
   }
  }
  User.HeightBlurRT {
   type=TextureRenderTarget
   description=""
   defaultValue {
    struct=NiagaraExt_VariableValue_Object
    value {object=/Game/Tools/3DHB/TextureBaker/RenderTarget/RT_3DHB_HeightBlurTexture, objectClass=&TextureRenderTarget}
   }
  }
  User.HeightRT {
   type=TextureRenderTarget
   description=""
   defaultValue {
    struct=NiagaraExt_VariableValue_Object
    value {object=/Game/Tools/3DHB/TextureBaker/RenderTarget/RT_3DHB_HeightTexture, objectClass=&TextureRenderTarget}
   }
  }
  User.HTN_B RT {
   type=TextureRenderTarget
   description=""
   defaultValue {
    struct=NiagaraExt_VariableValue_Object
    value {object=/Game/Tools/3DHB/TextureBaker/RenderTarget/RT_3DHB_HTN_Blur, objectClass=&TextureRenderTarget}
   }
  }
  User.MeshMaskRT {
   type=TextureRenderTarget
   description=""
   defaultValue {
    struct=NiagaraExt_VariableValue_Object
    value {object=/Game/Tools/3DHB/TextureBaker/RenderTarget/RT_3DHB_MeshMaskTexture, objectClass=&TextureRenderTarget}
   }
  }
  User.NormalRT {
   type=TextureRenderTarget
   description=""
   defaultValue {
    struct=NiagaraExt_VariableValue_Object
    value {object=/Game/Tools/3DHB/TextureBaker/RenderTarget/RT_3DHB_NormalTexture, objectClass=&TextureRenderTarget}
   }
  }
  User.ResU {type=NiagaraInt32, defaultValue=64}
  User.ResV {type=NiagaraInt32, defaultValue=64}
 ]
}
